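# Physics asset for the hip-area dynamic bones of the medium male Barbarian player body, covering skirt panels, belt buckle and belt teeth.
# Left and right skirt panels use one or more boxes and hang from the thigh twist bones, so they follow the legs.
# The center front panel and capsule belt pieces attach to the pelvis with limited swing and twist.
# Pelvis, spine and thigh capsules form the supporting chain, joined by free-rotation constraints.
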PhysicsAsset /Game/Characters/Player/Male/Medium/Bodies/M_MED_Barbarian/Meshes/M_MED_Barbarian_Dyn_Hips
bodies {
 dyn_belt_buckle [{shapeType=Capsule, length=7.9899, radius=3.8842, center=<0 0 1.5143>, rotation=<rot 90 0 0>}]
 dyn_belt_tooth_fr_l [{shapeType=Capsule, length=3, radius=3, center=<0 0 0.5>, rotation=<rot 90 0 0>}]
 dyn_belt_tooth_fr_r [{shapeType=Capsule, length=3, radius=3, center=<0 0 -0.5>, rotation=<rot 90 0 0>}]
 dyn_belt_tooth_l [{shapeType=Capsule, length=3, radius=3, center=<0 0 0.5>, rotation=<rot 90 0 0>}]
 dyn_belt_tooth_r [{shapeType=Capsule, length=3, radius=3, center=<0 0 -0.5>, rotation=<rot 90 0 0>}]
 dyn_skirt_bk_l_01 [{shapeType=Box, extentX=10.1875, extentY=21.5, extentZ=5.25, center=<5.4368 0 -0.0001>}]
 dyn_skirt_bk_r_01 [
  {shapeType=Box, extentX=17.875, extentY=21.5, extentZ=7.125, center=<-9.5929 0 -0.0002>}
  {shapeType=Box, extentX=4, extentY=18, extentZ=5.5, shapeName=dyn_skirt_bk_r_01_box, center=<-15.9194 -0.3784 6.2821>}
 ]
 dyn_skirt_fr_01 [{shapeType=Box, extentX=10.1875, extentY=10.9375, extentZ=1.1875, center=<5.4368 0 -0.0001>}]
 dyn_skirt_fr_l_01 [{shapeType=Box, extentX=14.0625, extentY=18, extentZ=2.5625, center=<8.2892 0 -0.0003>}]
 dyn_skirt_fr_r_01 [{shapeType=Box, extentX=15.75, extentY=18.875, extentZ=3.6875, center=<-8.0979 0 0.0001>}]
 dyn_skirt_out_l_01 [
  {shapeType=Box, extentX=16.6875, extentY=23.4375, extentZ=2.5, center=<8.9367 -0.0005 -0.0001>}
  {shapeType=Box, extentX=17.125, extentY=16.4375, extentZ=2.5, center=<8.034 7.1509 -4.8998>, rotation=<rot -8.1776 5.7753 54.5859>}
 ]
 dyn_skirt_out_r_01 [
  {shapeType=Box, extentX=21.1875, extentY=23.4375, extentZ=2.5, center=<-12.5467 -0.0004 -0.0001>}
  {shapeType=Box, extentX=11.6875, extentY=11.6875, extentZ=1.8125, center=<-18.4345 8.6192 5.1559>, rotation=<rot -10.5726 -3.6213 -44.2259>}
  {shapeType=Box, extentX=15.1875, extentY=13.75, extentZ=1.8125, center=<-14.0444 -5.791 4.0587>, rotation=<rot -4.8448 1.3006 40.1108>}
 ]
 pelvis [{shapeType=Capsule, length=3, radius=3}]
 spine_01 [{shapeType=Capsule, length=3, radius=3}]
 thigh_l [{shapeType=Capsule, length=26.1875, radius=8.8438, center=<-17.8608 0.0006 0.2433>, rotation=<rot 85.9471 180 -180>}]
 thigh_r [{shapeType=Capsule, length=26.1875, radius=8.8438, center=<17.7174 0.0001 -0.1974>, rotation=<rot 85.9418 180 -180>}]
 thigh_twist_01_l [{shapeType=Capsule, length=5.875, radius=2.9688, center=<3.9284 0 0.2052>, rotation=<rot 85.97 180 -180>}]
 thigh_twist_01_r [{shapeType=Capsule, length=5.875, radius=2.9688, center=<-3.9284 0 -0.2052>, rotation=<rot 85.97 180 -180>}]
}
constraints [
 {bone1Name=dyn_skirt_bk_l_01, bone2Name=thigh_twist_01_l, swing1LimitDegrees=60, swing2LimitDegrees=80, twistLimitDegrees=5}
 {bone1Name=dyn_skirt_fr_l_01, bone2Name=thigh_twist_01_l, swing1LimitDegrees=45, swing2LimitDegrees=45, twistLimitDegrees=5}
 {bone1Name=dyn_skirt_out_l_01, bone2Name=thigh_twist_01_l, swing1LimitDegrees=45, swing2LimitDegrees=45, twistLimitDegrees=5}
 {bone1Name=dyn_skirt_fr_r_01, bone2Name=thigh_twist_01_r, swing1LimitDegrees=45, swing2LimitDegrees=45, twistLimitDegrees=5}
 {bone1Name=dyn_skirt_out_r_01, bone2Name=thigh_twist_01_r, swing1LimitDegrees=45, swing2LimitDegrees=45, twistLimitDegrees=5}
 {bone1Name=dyn_skirt_bk_r_01, bone2Name=thigh_twist_01_r, swing1LimitDegrees=60, swing2LimitDegrees=80, twistLimitDegrees=5}
 {bone1Name=dyn_belt_buckle, bone2Name=pelvis, swing1LimitDegrees=5, swing2LimitDegrees=45, twistLimitDegrees=5}
 {bone1Name=dyn_belt_tooth_fr_l, bone2Name=pelvis, swing1LimitDegrees=5, swing2LimitDegrees=45, twistLimitDegrees=5}
 {bone1Name=dyn_belt_tooth_l, bone2Name=pelvis, swing1LimitDegrees=5, swing2LimitDegrees=45, twistLimitDegrees=5}
 {bone1Name=dyn_belt_tooth_fr_r, bone2Name=pelvis, swing1LimitDegrees=5, swing2LimitDegrees=45, twistLimitDegrees=5}
 {bone1Name=dyn_belt_tooth_r, bone2Name=pelvis, swing1LimitDegrees=5, swing2LimitDegrees=45, twistLimitDegrees=5}
 {
  bone1Name=thigh_twist_01_l
  bone2Name=thigh_l
  swing1LimitDegrees=45
  swing1Motion=Free
  swing2LimitDegrees=45
  swing2Motion=Free
  twistLimitDegrees=45
  twistMotion=Free
 }
 {
  bone1Name=thigh_twist_01_r
  bone2Name=thigh_r
  swing1LimitDegrees=45
  swing1Motion=Free
  swing2LimitDegrees=45
  swing2Motion=Free
  twistLimitDegrees=45
  twistMotion=Free
 }
 {bone1Name=dyn_skirt_fr_01, bone2Name=pelvis, swing1LimitDegrees=45, swing2LimitDegrees=45, twistLimitDegrees=5}
 {
  bone1Name=spine_01
  bone2Name=pelvis
  swing1LimitDegrees=45
  swing1Motion=Free
  swing2LimitDegrees=45
  swing2Motion=Free
  twistLimitDegrees=45
  twistMotion=Free
 }
 {
  bone1Name=thigh_l
  bone2Name=pelvis
  swing1LimitDegrees=45
  swing1Motion=Free
  swing2LimitDegrees=45
  swing2Motion=Free
  twistLimitDegrees=45
  twistMotion=Free
 }
 {
  bone1Name=thigh_r
  bone2Name=pelvis
  swing1LimitDegrees=45
  swing1Motion=Free
  swing2LimitDegrees=45
  swing2Motion=Free
  twistLimitDegrees=45
  twistMotion=Free
 }
]
