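# Physics asset for the medium male Barbarian's dynamic torso, defining bodies and joints for shoulder fur and a necklace.
# Capsules cover spine, clavicles and arm bones, linked by free angular constraints; flat boxes form the necklace and most fur panels.
# Angle-limited joints hang fur from spine_05, the right clavicle and arm, and chain the left ears, head and jaw to the left upper-arm twist.
# The necklace swings and twists freely on spine_05.
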PhysicsAsset /Game/Characters/Player/Male/Medium/Bodies/M_MED_Barbarian/Meshes/M_MED_Barbarian_Dyn_Torso
bodies {
 clavicle_l [{shapeType=Capsule, length=3, radius=3}]
 clavicle_r [{shapeType=Capsule, length=3, radius=3}]
 dyn_necklace_02 [{shapeType=Box, extentX=18.625, extentY=10.9375, extentZ=1.1875, center=<9.9049 0 -0.0001>}]
 dyn_shoulderfur_bk [{shapeType=Box, extentX=10.1875, extentY=10.9375, extentZ=1.1875, center=<5.4368 0 -0.0001>}]
 dyn_shoulderfur_bk_l [{shapeType=Box, extentX=10.1875, extentY=10.9375, extentZ=1.6875, center=<5.4368 0 -0.0001>}]
 dyn_shoulderfur_bk_r [{shapeType=Box, extentX=10.1875, extentY=10.9375, extentZ=1.1875, center=<5.4368 0 -0.0001>}]
 dyn_shoulderfur_ear_left [{shapeType=Capsule, length=3, radius=3, center=<0.3 0 0>, rotation=<rot 90 0 0>}]
 dyn_shoulderfur_ear_right [{shapeType=Capsule, length=3, radius=3, center=<0.3 0 0>, rotation=<rot 90 0 0>}]
 dyn_shoulderfur_fr_r [{shapeType=Box, extentX=10.1875, extentY=10.9375, extentZ=1.1875, center=<5.4368 0 -0.0001>}]
 dyn_shoulderfur_head_l [{shapeType=Capsule, length=3, radius=3, center=<0.5 0 0>, rotation=<rot 90 0 0>}]
 dyn_shoulderfur_jaw_l [{shapeType=Capsule, length=11.125, radius=2.5312, center=<1.5696 0 0>, rotation=<rot 90 0 0>}]
 dyn_shoulderfur_r [{shapeType=Box, extentX=10.1875, extentY=10.9375, extentZ=1.1875, center=<5.4368 0 -0.0001>}]
 dyn_shoulderfur_r_tip [{shapeType=Capsule, length=3, radius=3, rotation=<rot 90 0 0>}]
 spine_03 [{shapeType=Capsule, length=16.7282, radius=9.8521, center=<4.6476 3.3197 -0.0675>, rotation=<rot 1.1877 -90 -0.4746>}]
 spine_04 [{shapeType=Capsule, length=15.9157, radius=10.6646, center=<2.4069 4.3335 -0.1074>, rotation=<rot 1.1877 -90 -0.4746>}]
 spine_05 [
  {shapeType=Capsule, length=18, radius=10.9688, center=<-3.845 1.9009 0>}
  {shapeType=Capsule, length=18.9375, radius=10.3125, center=<-0.7931 -2.9074 0.1384>, rotation=<rot 0 0 2.8567>}
 ]
 upperarm_r [{shapeType=Capsule, length=19.3354, radius=6.4575, center=<-14.2078 -1.2706 0.8678>, rotation=<rot 90 0 0>}]
 upperarm_twist_01_l [{shapeType=Capsule, length=2, radius=2, rotation=<rot 90 0 0>}]
 upperarm_twist_01_r [{shapeType=Capsule, length=2, radius=2}]
 upperarm_twist_02_r [{shapeType=Capsule, length=2, radius=2}]
}
constraints [
 {bone1Name=dyn_shoulderfur_head_l, bone2Name=dyn_shoulderfur_jaw_l, swing1LimitDegrees=3, swing2LimitDegrees=15, twistLimitDegrees=3}
 {bone1Name=dyn_shoulderfur_ear_left, bone2Name=dyn_shoulderfur_head_l, swing1LimitDegrees=10, swing2LimitDegrees=45, twistLimitDegrees=3}
 {bone1Name=dyn_shoulderfur_ear_right, bone2Name=dyn_shoulderfur_head_l, swing1LimitDegrees=10, swing2LimitDegrees=45, twistLimitDegrees=3}
 {bone1Name=dyn_shoulderfur_jaw_l, bone2Name=upperarm_twist_01_l, swing1LimitDegrees=10, swing2LimitDegrees=15, twistLimitDegrees=2}
 {
  bone1Name=clavicle_l
  bone2Name=spine_05
  swing1LimitDegrees=45
  swing1Motion=Free
  swing2LimitDegrees=45
  swing2Motion=Free
  twistLimitDegrees=45
  twistMotion=Free
 }
 {
  bone1Name=clavicle_r
  bone2Name=spine_05
  swing1LimitDegrees=45
  swing1Motion=Free
  swing2LimitDegrees=45
  swing2Motion=Free
  twistLimitDegrees=45
  twistMotion=Free
 }
 {bone1Name=dyn_shoulderfur_fr_r, bone2Name=upperarm_twist_01_r, swing1LimitDegrees=45, swing2LimitDegrees=15, twistLimitDegrees=2}
 {bone1Name=dyn_shoulderfur_r, bone2Name=upperarm_twist_01_r, swing1LimitDegrees=45, swing2LimitDegrees=30, twistLimitDegrees=2}
 {bone1Name=dyn_shoulderfur_r_tip, bone2Name=clavicle_r, swing1LimitDegrees=10, swing2LimitDegrees=45, twistLimitDegrees=3}
 {bone1Name=dyn_shoulderfur_bk_l, bone2Name=spine_05, swing1LimitDegrees=15, swing2LimitDegrees=5, twistLimitDegrees=2}
 {bone1Name=dyn_shoulderfur_bk, bone2Name=spine_05, swing1LimitDegrees=15, swing2LimitDegrees=5, twistLimitDegrees=2}
 {bone1Name=dyn_shoulderfur_bk_r, bone2Name=spine_05, swing1LimitDegrees=45, swing2LimitDegrees=20, twistLimitDegrees=2}
 {
  bone1Name=dyn_necklace_02
  bone2Name=spine_05
  swing1LimitDegrees=45
  swing1Motion=Free
  swing2LimitDegrees=45
  swing2Motion=Free
  twistLimitDegrees=45
  twistMotion=Free
 }
 {
  bone1Name=spine_05
  bone2Name=spine_04
  swing1LimitDegrees=45
  swing1Motion=Free
  swing2LimitDegrees=45
  swing2Motion=Free
  twistLimitDegrees=45
  twistMotion=Free
 }
 {
  bone1Name=spine_04
  bone2Name=spine_03
  swing1LimitDegrees=45
  swing1Motion=Free
  swing2LimitDegrees=45
  swing2Motion=Free
  twistLimitDegrees=45
  twistMotion=Free
 }
 {
  bone1Name=upperarm_twist_01_l
  bone2Name=clavicle_l
  swing1LimitDegrees=45
  swing1Motion=Free
  swing2LimitDegrees=45
  swing2Motion=Free
  twistLimitDegrees=45
  twistMotion=Free
 }
 {
  bone1Name=upperarm_r
  bone2Name=clavicle_r
  swing1LimitDegrees=45
  swing1Motion=Free
  swing2LimitDegrees=45
  swing2Motion=Free
  twistLimitDegrees=45
  twistMotion=Free
 }
 {
  bone1Name=upperarm_twist_01_r
  bone2Name=upperarm_r
  swing1LimitDegrees=45
  swing1Motion=Free
  swing2LimitDegrees=45
  swing2Motion=Free
  twistLimitDegrees=45
  twistMotion=Free
 }
 {
  bone1Name=upperarm_twist_02_r
  bone2Name=upperarm_r
  swing1LimitDegrees=45
  swing1Motion=Free
  swing2LimitDegrees=45
  swing2Motion=Free
  twistLimitDegrees=45
  twistMotion=Free
 }
]
