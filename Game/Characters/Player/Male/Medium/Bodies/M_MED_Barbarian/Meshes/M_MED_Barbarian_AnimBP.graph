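# Animation Blueprint for the medium male Barbarian body that follows the attached parent's pose and adds secondary motion to fur, skirt and necklace bones.
# The copied pose passes through shoulder-fur constraints, torso and hips RigidBody simulations, then a front-skirt constraint.
# Spring-based AnimDynamics nodes move the necklace and half-weighted wrist and calf fur, before additive offsets adjust two right skirt bones.
# The event graph has no logic, and a back-skirt constraint is left unconnected.
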AnimBlueprint /Game/Characters/Player/Male/Medium/Bodies/M_MED_Barbarian/Meshes/M_MED_Barbarian_AnimBP
target_skeleton=&/M_MED_Barbarian_Skeleton
graphs [
 AnimGraph {nodes [
   "Constraint - dyn_skirt_bk_r_01 (PP)"!(ComponentPose, Alpha=1, ConstraintWeights_0=1, ConstraintWeights_1:float=0) -> (Pose) {
    alphaBoolBlend {blendOption=Linear}
    boneToModify {boneName=dyn_skirt_bk_r_01}
    constraintSetup [{targetBone {boneName=thigh_r}}, {targetBone {boneName=pelvis}}]
    constraintWeights [1,1]
   }
   n1 = "Constraint - dyn_shoulderfur_r (PPPP)"!(
    ComponentPose=Local To Component(LocalPose=Copy Pose From Mesh() {bUseAttachedParent=true})
    Alpha=1
    ConstraintWeights_0=0.3136
    ConstraintWeights_1:float=0.6854
    ConstraintWeights_2:float=0.0003
    ConstraintWeights_3:float=0.0007
   ) -> (Pose) {
    alphaBoolBlend {blendOption=Linear}
    boneToModify {boneName=dyn_shoulderfur_r}
    constraintSetup [
     {targetBone {boneName=clavicle_r}}
     {targetBone {boneName=upperarm_twist_01_r}}
     {targetBone {boneName=upperarm_r}}
     {targetBone {boneName=upperarm_twist_02_r}}
    ]
    constraintWeights [1,1,1,1]
   }
   n2 = "Constraint - dyn_shoulderfur_fr_r (PPPPPP)"!(
    ComponentPose=@n1
    Alpha=1
    ConstraintWeights_0=0
    ConstraintWeights_1:float=0.4958
    ConstraintWeights_2:float=0.0001
    ConstraintWeights_3:float=0.0007
    ConstraintWeights_4:float=0.0052
    ConstraintWeights_5:float=0.4982
   ) -> (Pose) {
    alphaBoolBlend {blendOption=Linear}
    boneToModify {boneName=dyn_shoulderfur_fr_r}
    constraintSetup [
     {targetBone {boneName=upperarm_r}}
     {targetBone {boneName=upperarm_twist_01_r}}
     {targetBone {boneName=upperarm_twist_02_r}}
     {targetBone {boneName=spine_03}}
     {targetBone {boneName=spine_04}}
     {targetBone {boneName=clavicle_r}}
    ]
    constraintWeights [1,1,1,1,1,1]
   }
   n3 = "Constraint - dyn_shoulderfur_bk_r (TT)"!(ComponentPose=@n2, Alpha=1, ConstraintWeights_0=1, ConstraintWeights_1:float=1) -> (Pose) {
    alphaBoolBlend {blendOption=Linear}
    boneToModify {boneName=dyn_shoulderfur_bk_r}
    constraintSetup [{transformType=Translation, targetBone {boneName=dyn_shoulderfur_r}}, {transformType=Translation, targetBone {boneName=dyn_shoulderfur_bk}}]
    constraintWeights [1,1]
   }
   n4 = RigidBody(ComponentPose=@n3) {
    bOverrideWorldGravity=true
    overridePhysicsAsset=&/M_MED_Barbarian_Dyn_Torso
    simulationSpace=BaseBoneSpace
    alphaBoolBlend {blendOption=Linear}
    baseBoneRef {boneName=spine_05}
    componentAppliedLinearAccClamp=<2000 2000 2000>
    componentLinearAccScale=<0.45 0.25 0.25>
    componentLinearVelScale=<0.45 0.25 0.25>
    overrideWorldGravity {z=-980}
   }
   n5 = RigidBody(ComponentPose=@n4) {
    bOverrideWorldGravity=true
    overridePhysicsAsset=&/M_MED_Barbarian_Dyn_Hips
    simulationSpace=BaseBoneSpace
    alphaBoolBlend {blendOption=Linear}
    baseBoneRef {boneName=pelvis}
    componentAppliedLinearAccClamp=<2000 2000 2000>
    componentLinearAccScale=<0.8 0.5 0.5>
    componentLinearVelScale=<0.8 0.5 0.5>
    overrideWorldGravity {z=-980}
   }
   n6 = "Constraint - dyn_skirt_fr_01 (PP)"!(ComponentPose=@n5, Alpha=1, ConstraintWeights_0=1, ConstraintWeights_1:float=1) -> (Pose) {
    alphaBoolBlend {blendOption=Linear}
    boneToModify {boneName=dyn_skirt_fr_01}
    constraintSetup [{targetBone {boneName=dyn_skirt_fr_l_01}}, {targetBone {boneName=dyn_skirt_fr_r_01}}]
    constraintWeights [1,1]
   }
   n7 = Anim Dynamics - Bone: dyn_necklace_02(ComponentPose=@n6, Alpha=1, ExternalForce="0,0,0") {
    angularDampingOverride=0.998
    angularSpringConstant=70
    bAngularSpring=true
    bGravityOverrideInSimSpace=true
    bOverrideAngularDamping=true
    bUseGravityOverride=true
    bUsePlanarLimit=false
    bUseSphericalLimits=true
    simulationSpace=RootRelative
    alphaBoolBlend {blendOption=Linear}
    boundBone {boneName=dyn_necklace_02}
    componentAppliedLinearAccClamp=<3000 3000 3000>
    componentLinearAccScale=<0.3 0.3 0.3>
    componentLinearVelScale=<0.3 0.3 0.3>
    gravityOverride {z=-980}
    physicsBodyDefinitions [{
      collisionType=InnerSphere
      sphereCollisionRadius=1
      boundBone {boneName=dyn_necklace_02}
      boxExtents=<7 12 12>
      constraintSetup {twistAxis=AxisZ, angularLimitsMax=<0 0 5>, angularLimitsMin=<0 -35 -5>, angularTarget=<0 0 0>}
      localJointOffset=<2 0 0>
     }]
    relativeSpaceBone {boneName=root}
   }
   n8 = Anim Dynamics - Bone: dyn_wrist_fur_l(ComponentPose=@n7, Alpha=0.5, ExternalForce="0,0,0") {
    angularDampingOverride=0.98
    angularSpringConstant=15000
    bAngularSpring=true
    bGravityOverrideInSimSpace=true
    bOverrideAngularDamping=true
    simulationSpace=RootRelative
    alphaBoolBlend {blendOption=Linear}
    boundBone {boneName=dyn_wrist_fur_l}
    physicsBodyDefinitions [{
      sphereCollisionRadius=0
      boundBone {boneName=dyn_wrist_fur_l}
      boxExtents=<10 30 30>
      constraintSetup {angularLimitsMax=<30 30 30>, angularLimitsMin=<-30 -30 -30>, angularTarget=<1 0 0>}
      localJointOffset=<0.5 0 0>
     }]
   }
   n9 = Anim Dynamics - Bone: dyn_wrist_fur_r(ComponentPose=@n8, Alpha=0.5, ExternalForce="0,0,0") {
    angularDampingOverride=0.98
    angularSpringConstant=15000
    bAngularSpring=true
    bGravityOverrideInSimSpace=true
    bOverrideAngularDamping=true
    simulationSpace=RootRelative
    alphaBoolBlend {blendOption=Linear}
    boundBone {boneName=dyn_wrist_fur_r}
    physicsBodyDefinitions [{
      sphereCollisionRadius=0
      boundBone {boneName=dyn_wrist_fur_r}
      boxExtents=<10 30 30>
      constraintSetup {angularLimitsMax=<30 30 30>, angularLimitsMin=<-30 -30 -30>, angularTarget=<1 0 0>}
      localJointOffset=<-0.5 0 0>
     }]
   }
   n10 = Anim Dynamics - Bone: dyn_calf_fur_l(ComponentPose=@n9, Alpha=0.5, ExternalForce="0,0,0") {
    angularDampingOverride=0.98
    angularSpringConstant=15000
    bAngularSpring=true
    bGravityOverrideInSimSpace=true
    bOverrideAngularDamping=true
    simulationSpace=RootRelative
    alphaBoolBlend {blendOption=Linear}
    boundBone {boneName=dyn_calf_fur_l}
    physicsBodyDefinitions [{
      sphereCollisionRadius=0
      boundBone {boneName=dyn_calf_fur_l}
      boxExtents=<10 30 30>
      constraintSetup {angularLimitsMax=<30 30 30>, angularLimitsMin=<-30 -30 -30>, angularTarget=<1 0 0>}
      localJointOffset=<0.5 0 0>
     }]
   }
   n11 = Anim Dynamics - Bone: dyn_calf_fur_r(ComponentPose=@n10, Alpha=0.5, ExternalForce="0,0,0") {
    angularDampingOverride=0.98
    angularSpringConstant=15000
    bAngularSpring=true
    bGravityOverrideInSimSpace=true
    bOverrideAngularDamping=true
    simulationSpace=RootRelative
    alphaBoolBlend {blendOption=Linear}
    boundBone {boneName=dyn_calf_fur_r}
    physicsBodyDefinitions [{
      sphereCollisionRadius=0
      boundBone {boneName=dyn_calf_fur_r}
      boxExtents=<10 30 30>
      constraintSetup {angularLimitsMax=<30 30 30>, angularLimitsMin=<-30 -30 -30>, angularTarget=<1 0 0>}
      localJointOffset=<-0.5 0 0>
     }]
   }
   n12 = "Transform (Modify) Bone - Bone: dyn_skirt_bk_r_01"(ComponentPose=@n11, Alpha=1, Translation="-1,-2,0", Rotation="0,0,10", Scale="1,1,1") {
    rotationMode=BMM_Additive
    translationMode=BMM_Additive
    alphaBoolBlend {blendOption=Linear}
    boneToModify {boneName=dyn_skirt_bk_r_01}
    rotation {roll=10}
    translation=<-1 -2>
   }
   n13 = "Transform (Modify) Bone - Bone: dyn_skirt_out_r_01"(ComponentPose=@n12, Alpha=1, Translation="-2,-1,0", Rotation="0,0,0", Scale="1,1,1") {
    rotationMode=BMM_Additive
    translationMode=BMM_Additive
    alphaBoolBlend {blendOption=Linear}
    boneToModify {boneName=dyn_skirt_out_r_01}
    translation=<-2 -1>
   }
   Output Pose(Result=Component To Local(ComponentPose=@n13)) {layerGroup=None}
  ]}
 EventGraph {nodes [Event BlueprintUpdateAnimation()]}
]
variables []
